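# Sound cue for the player door bash impact in Fort_Player foley.
# A Random node picks one of five impact wave variations with equal weight and is excluded from branch culling.
# It plays boosted in volume through basic foley attenuation and the player foley sound class.
# Overridden concurrency allows one instance per owner and blocks new plays while one is active.
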SoundCue /FNGameplayCues/Sounds/Fort_Player/DoorBash/Door_Bash_Impact_Cue
root=Random_0
nodes [
 {
  id=Random_0
  children [WavePlayer_0,WavePlayer_1,WavePlayer_2,WavePlayer_3,WavePlayer_4]
  properties {bShouldExcludeFromBranchCulling=true, weights [1,1,1,1,1]}
 }
 {id=WavePlayer_0, properties {soundWaveAssetPtr=&/Door_Bash_Impact_05}}
 {id=WavePlayer_1, properties {soundWaveAssetPtr=&/Door_Bash_Impact_01}}
 {id=WavePlayer_2, properties {soundWaveAssetPtr=&/Door_Bash_Impact_02}}
 {id=WavePlayer_3, properties {soundWaveAssetPtr=&/Door_Bash_Impact_03}}
 {id=WavePlayer_4, properties {soundWaveAssetPtr=&/Door_Bash_Impact_04}}
]
properties {
 attenuationSettings=&Foley_Attenuation_Basic_Att
 bOverrideConcurrency=true
 maxDistance=1536
 soundClassObject=&Fort_Player_Foley
 volumeMultiplier=2
 concurrencyOverrides {bLimitToOwner=true, maxCount=1, resolutionRule=PreventNew}
}
