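# Sound cue for the player's door bash impulse, routed to the player foley sound class with basic foley attenuation.
# It randomly picks one of two equally weighted impulse waves and varies its pitch upward slightly through a modulator.
# A short fixed delay at the root holds back playback each time the cue fires.
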SoundCue /FNGameplayCues/Sounds/Fort_Player/DoorBash/Door_Bash_Player_Impulse_Cue
root=Delay_0
nodes [
 {id=Delay_0, children [Modulator_0], properties {delayMax=0.15, delayMin=0.15}}
 {id=Modulator_0, children [Random_0], properties {pitchMax=1.2, pitchMin=1}}
 {id=Random_0, children [WavePlayer_0,WavePlayer_1], properties {weights [1,1]}}
 {id=WavePlayer_0, properties {soundWaveAssetPtr=&/Door_Bash_Player_Impulse_02}}
 {id=WavePlayer_1, properties {soundWaveAssetPtr=&/Door_Bash_Player_Impulse_01}}
]
properties {
 attenuationSettings=&Foley_Attenuation_Basic_Att
 maxDistance=1536
 soundClassObject=&Fort_Player_Foley
 volumeMultiplier=1
}
